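# M_RandomFloatFlicker_Example_PARENT is a Default Lit example parent material in the MaterialGym folder that demonstrates a random float flicker.
# The RandomFlickerFloat output of MF__FloatFlickerRandom_Generic, with none of its inputs wired, drives Base Color through a SetMaterialAttributes node.
# The material uses material attributes, and the Emissive Color slot on that node is left unconnected.
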Material /Game/Global/Materials/MaterialGym/Materials/RandomFloatFlicker/M_RandomFloatFlicker_Example_PARENT
graph {
 properties {bUseMaterialAttributes=true, pixelDepthOffsetMode=PDOM_Legacy, usedShadingModels=Default Lit}
 nodes [SetMaterialAttributes!(
   MaterialAttributes
   Emissive Color
   Base Color=MaterialFunctionCall() {materialFunction=/Game/MF__FloatFlickerRandom_Generic}.RandomFlickerFloat
  ) {attributeSetTypes [#1, #2]}]
}
